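; Blueprint for the Cerberus spin-slash pickaxe, a melee harvesting tool that is a child of a parent weapon class.
; Swing events and the impact FX event only forward to the parent, as does the construction script.
; BeginPlay and Tick are empty stubs, so this graph adds no logic of its own.
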(event EventBeginPlay)

(event EventTick
  (params (DeltaSeconds :type Float%20%28single-precision%29)))

(event EventSwingLeft2
  (|CallParentFunction))

(event FortWeapon|EventOnPlayImpactFX
  (params (HitResult :type Hit%20Result%20Structure%20%28by%20ref%29) (ImpactPhysicalSurface :type EPhysicalSurface%20Enum :default "SurfaceType_Default") (SpawnedPSC :type FXSystem%20Component%20Object%20Reference))
  (|CallParentFunction HitResult ImpactPhysicalSurface SpawnedPSC))

(event EventSwingLeft_Common
  (|CallParentFunction))

(event EventSwingRight_Common
  (|CallParentFunction))

(event EventSwingLeft
  (|CallParentFunction))

(event EventSwingRight
  (|CallParentFunction))

(fn ConstructionScript
  (|CallParentFunction))
